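# Sound cue for the looping sound of the Black Monday male glider cosmetic.
# A looping wave player feeds a modulator that slightly randomizes volume.
# The cue uses player glider thrust attenuation, a cosmetic glider loop audio properties sheet and the gliders land sound class, with a reduced overall volume multiplier.
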SoundCue /BRCosmetics/Sounds/Gliders/BlackMondayMale/Glider_BlackMonday_Male_Loop_Cue
root=Modulator_0
nodes [
 {id=Modulator_0, children [WavePlayer_0], properties {volumeMax=0.8, volumeMin=0.7}}
 {id=WavePlayer_0, properties {bLooping=true, soundWaveAssetPtr=&/Glider_BlackMonday_Male_Loop_01}}
]
properties {
 attenuationSettings=&Fort_Player_Glider_Thrust_Athena_Att
 audioPropertiesSheet=&APS_Cosmetic_Glider_Loop
 maxDistance=12500
 soundClassObject=&Fort_Gliders_Land
 volumeMultiplier=0.32
}
