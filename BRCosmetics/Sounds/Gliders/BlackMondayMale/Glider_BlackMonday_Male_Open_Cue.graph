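# Sound cue for the opening of the male Black Monday glider cosmetic.
# After a short initial delay, a mixer layers the main open sound, a later open layer and a still later, slightly louder thrust, staggered by per-branch delays.
# Glider open/close attenuation and concurrency, a glider sound class and a shared glider-open audio property sheet govern playback.
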SoundCue /BRCosmetics/Sounds/Gliders/BlackMondayMale/Glider_BlackMonday_Male_Open_Cue
root=Delay_2
nodes [
 {id=Delay_0, children [WavePlayer_2], properties {delayMax=1, delayMin=1}}
 {id=Delay_1, children [WavePlayer_0], properties {delayMax=1.6, delayMin=1.6}}
 {id=Delay_2, children [Mixer_0], properties {delayMax=0.34, delayMin=0.34}}
 {id=Mixer_0, children [WavePlayer_1,Delay_0,Delay_1], properties {inputVolume [1,1,1.1]}}
 {id=WavePlayer_0, properties {soundWaveAssetPtr=&/Glider_BlackMonday_Male_Thrust_01}}
 {id=WavePlayer_1, properties {soundWaveAssetPtr=&/Glider_BlackMonday_Male_Open_01}}
 {id=WavePlayer_2, properties {soundWaveAssetPtr=&/Glider_BlackMonday_Male_OpenLayer_01}}
]
properties {
 attenuationSettings=/Game/Sounds/Attenuation/Fort_Player_Glider_OpenClose_Athena_Att
 audioPropertiesSheet=/BRCosmetics/Sounds/Settings/PropertySheets/Glider/APS_Cosmetic_Glider_Open
 maxDistance=10000
 soundClassObject=&Fort_Gliders_Land
 volumeMultiplier=0.6
 concurrencySet [&Athena_GliderOpenClose_Conc]
}
